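# Translucent lit Niagara sprite material for a particle-tinted water splash, in the CyberFuGlitch glider's FX folder.
# It reuses a blood smear decal texture and normal for shape, with panned normal-map UV distortion and Dynamic Parameters for randomness, foam, distortion and erosion.
# Opacity combines eroded mask, depth fade, near-camera fade and a dot-product falloff.
# Branches build noise-masked foam color, refraction, emissive, and fake sun specular with a mobile minimum.
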Material /Game/Gadgets/Assets/VinderTech_GliderChute/Glider_CyberFuGlitch/FX/M_Water_Splash_Color
graph {
 properties {
  bUsedWithNiagaraSprites=true
  blendMode=BLEND_Translucent
  pixelDepthOffsetMode=PDOM_Legacy
  refractionMethod=RM_IndexOfRefraction
  translucentSelfShadowDensityScale=0.1
  translucentShadowDensityScale=0.1
  usedShadingModels=Default Lit
 }
 nodes [
  n1 = NamedRerouteUsage()
  n2 = Constant() {r=0.5}
  n3 = Add(A=Multiply(A=Transform(None=@n1) {transformType=TRANSFORM_View}, B=@n2), B=@n2)
  n4 = Multiply(
   A=TextureSample(UVs=ComponentMask(None=@n3) {g=true, r=true}) {samplerType=SAMPLERTYPE_LinearColor, texture=/Game/Effects/Fort_Effects/Textures/BloodSmearDecal_13}.RGB
   B=ScalarParameter() {defaultValue=0.2, group=Specular, parameterName=Reflection Amount}
  )
  n5 = ScalarParameter() {
   defaultValue=0.075
   desc="Offset the Minimum fake specular amount for Mobile.  ("
   group=Specular
   parameterName=Mobile Specular Min Offset
   sliderMax=0.2
  }
  n6 = FeatureLevelSwitch(Default=Constant(), ES3_1=@n5)
  n7 = CollectionParameter() {
   collection=&FortniteMaterialParameters
   parameterName=SunAndMoonModelDirectionalVector
  }
  n8 = ComponentMask(None=@n7) {b=true, g=true, r=true}
  n9 = ReflectionVectorWS(CustomWorldNormal=Transform(None=@n1))
  n10 = NamedRerouteUsage()
  n11 = Power(
   Base=Clamp(None=DotProduct(A=@n8, B=@n9))
   Exp=LinearInterpolate(
    A=ScalarParameter() {defaultValue=40, group=Specular, parameterName=Specular Power}
    B=ScalarParameter() {defaultValue=5, group=Foam, parameterName=Foam - Specular Power}
    Alpha=@n10
   )
  ) {constExponent=40}
  n12 = LinearInterpolate(
   A=ScalarParameter() {defaultValue=5, group=Specular, parameterName=Specular Intensity}
   B=ScalarParameter() {defaultValue=1, group=Foam, parameterName=Foam - Specular Intensity}
   Alpha=@n10
  )
  n13 = Power(Base=ComponentMask(None=CollectionParameter() {collection=&FortniteMaterialParameters, parameterName=SunLightColor}) {b=true, g=true, r=true}) {constExponent=0.1}
  Add(
   A=Multiply(
    A=Clamp(None=@n4, Min=@n6) {minDefault=0.1}
    B=MaterialFunctionCall() {materialFunction=&Color_Time_of_Day}.Result
   )
   B=Multiply(A=Clamp(None=@n11, Min=@n6, Max=@n12), B=Clamp(None=@n13))
  ) {constB=0}
  ComponentMask() {g=true, r=true}
  NamedRerouteDeclaration(None=Multiply(
    A=TextureSample() {samplerType=SAMPLERTYPE_LinearColor, texture=&BloodSmearDecal_13}.RGB
    B=ParticleColor().RGB
   )) {name=Emissive, nodeColor=<rgba 0 0.6667 1 1>}
  n14 = TextureSampleParameter2D(UVs=NamedRerouteUsage()) {
   parameterName=Opacity Mask
   samplerSource=SSM_Clamp_WorldGroupSettings
   samplerType=SAMPLERTYPE_LinearColor
   texture=&BloodSmearDecal_13
  }
  n15 = DynamicParameter() {defaultValue=<rgba 1 0 1 0>, paramNames [Rand 0-1, Foam, UV Distortion, Erode]}
  n16 = Saturate(None=Divide(A=Subtract(A=@n14.G, B=@n15.Erode), B=ScalarParameter() {defaultValue=0.1, group=Opacity, parameterName=OpacitySoftness}))
  n17 = ScalarParameter() {defaultValue=2, group=Opacity, parameterName=Depth Fade Distance}
  n18 = Power(
   Base=MaterialFunctionCall(Distance=ScalarParameter() {defaultValue=40, group=Opacity, parameterName=NearFade}) {materialFunction=&Near_Camera_Fade}.Result
   Exp=ScalarParameter() {defaultValue=3, group=Opacity, parameterName=NearFadeExp}
  )
  n19 = Multiply(A=Multiply(A=DepthFade(Opacity=@n16, FadeDistance=@n17), B=Saturate(None=@n18)), B=ParticleColor().A)
  n20 = DotProduct(A=NamedRerouteUsage(), B=Constant3Vector() {constant=<rgba 0 0 1 0>})
  n21 = LinearInterpolate(
   A=LinearInterpolate(A=ScalarParameter() {defaultValue=1, group=Opacity, parameterName=Fresnel Minimum Opacity}, Alpha=NamedRerouteUsage())
   Alpha=OneMinus(None=Subtract(A=@n20) {constB=0.15})
  ) {constA=0.6}
  NamedRerouteDeclaration(None=Saturate(None=Multiply(A=@n19, B=@n21))) {name=Opacity, nodeColor=<rgba 1 1 1 1>}
  n22 = Multiply(
   A=LinearInterpolate(
    A=VectorParameter() {group=Color, parameterName=Light Color, defaultValue=<rgba 0.33 0.4675 0.66 0>}.RGB
    B=VectorParameter() {group=Color, parameterName=Dark Color, defaultValue=<rgba 0.2025 0.3056 0.45 0>}.RGB
    Alpha=NamedRerouteUsage()
   ) {constA=1, constB=0.75}
   B=ParticleColor().RGB
  )
  n23 = Desaturation(None=@n22, Fraction=ScalarParameter() {defaultValue=0.66, group=Foam, parameterName=Foam - Desaturation})
  n24 = ScalarParameter() {defaultValue=10, group=Foam, parameterName=Foam - Value Multiplier}
  NamedRerouteDeclaration(None=LinearInterpolate(A=@n22, B=Multiply(A=@n23, B=@n24), Alpha=NamedRerouteUsage())) {name=Diffuse Color, nodeColor=<rgba 1 0.25 0 1>}
  n25 = Multiply(A=TextureCoordinate(), B=ScalarParameter() {defaultValue=8, group=UV, parameterName=Distortion Tile})
  n26 = Add(A=DynamicParameter() {defaultValue=<rgba 1 0 1 0>, paramNames [Rand 0-1, Foam, UV Distortion, Erode]}.Rand 0-1, B=ParticleRelativeTime())
  n27 = Multiply(A=Constant2Vector() {g=-0.8, r=0.05}, B=ScalarParameter() {defaultValue=1, group=UV, parameterName=Distortion Speed})
  n28 = ComponentMask(None=TextureSample(UVs=Panner(Coordinate=@n25, Time=@n26, Speed=@n27) {speedX=0.05, speedY=0.5}) {samplerType=SAMPLERTYPE_Normal, texture=&T_Weap_Distort_N_02}.RGB) {g=true, r=true}
  n29 = DynamicParameter() {defaultValue=<rgba 1 0 1 0>, paramNames [Rand 0-1, Foam, UV Distortion, Erode]}
  n30 = Multiply(
   A=Multiply(A=ScalarParameter() {defaultValue=0.02, group=UV, parameterName=Distortion Amount}, B=@n29.UV Distortion)
   B=LinearInterpolate(Alpha=@n29.Rand 0-1) {constA=0.5, constB=1.2}
  )
  NamedRerouteDeclaration(None=Add(A=TextureCoordinate(), B=Multiply(A=@n28, B=@n30))) {name=UVs, nodeColor=<rgba 1 1 0 1>}
  n31 = TextureSampleParameter2D(UVs=NamedRerouteUsage()) {parameterName=Normal, samplerType=SAMPLERTYPE_Normal, texture=/Game/Effects/Fort_Effects/Textures/BloodSmearDecal_13_N}
  n32 = DynamicParameter() {defaultValue=<rgba 1 0 1 0>, paramNames [Rand 0-1, Foam, UV Distortion, Erode]}
  n33 = Subtract(A=@n31.B, B=@n32.Erode)
  n34 = OneMinus(None=@n32.Erode)
  n35 = MaterialFunctionCall(X=@n31.R, Y=@n31.R, Z=Saturate(None=Divide(A=@n33, B=@n34))) {materialFunction=&MakeFloat3}
  NamedRerouteDeclaration(None=MaterialFunctionCall(Normal=Normalize(VectorInput=@n35.Result), Flatness=ScalarParameter() {group=Normal, parameterName=FlattenNormal}) {materialFunction=&FlattenNormal}.Result) {name=Normal, nodeColor=<rgba 0 0.25 1 1>}
  NamedRerouteDeclaration(None=LinearInterpolate(B=ScalarParameter() {defaultValue=0.95, group=Refraction, parameterName=Refraction Value}, Alpha=NamedRerouteUsage()) {constA=1}) {name=Refraction, nodeColor=<rgba 1 1 1 1>}
  n36 = DynamicParameter() {defaultValue=<rgba 1 0 1 0>, paramNames [Rand 0-1, Foam, UV Distortion, Erode]}
  n37 = TextureSampleParameter2D(UVs=Multiply(A=TextureCoordinate(), B=ScalarParameter() {defaultValue=2.5, group=Foam, parameterName=Foam - Mask Tile})) {
   group=Foam
   parameterName=Foam Mask
   samplerType=SAMPLERTYPE_LinearGrayscale
   texture=/Game/Global/Textures/Default/Noise/T_SwirlNoise_N_height
  }
  n38 = Divide(
   A=Add(
    A=LinearInterpolate(Alpha=@n36.Foam) {constA=-1}
    B=ChannelMaskParameter(None=@n37.RGB) {group=Foam, maskChannel=Green, parameterName=Foam Mask Channel, defaultValue=<rgba 0 1 0 0>}.R
   )
   B=ScalarParameter() {defaultValue=0.75, group=Foam, parameterName=Foam - Mask Softness}
  ) {constB=0.25}
  NamedRerouteDeclaration(None=LinearInterpolate(B=Saturate(None=@n38), Alpha=ScalarParameter() {group=Foam, parameterName=Enable Foam})) {name=Foam, nodeColor=<rgba 0 0.2734 1 1>}
  NamedRerouteUsage()
  NamedRerouteUsage()
  NamedRerouteUsage()
  NamedRerouteUsage()
  NamedRerouteUsage()
  NamedRerouteUsage()
  ScalarParameter() {defaultValue=1, group=Opacity, parameterName=Fresnel Minimum Opacity}
  VectorParameter() {group=Color, parameterName=Subsurface Color, defaultValue=<rgba 0.25 0.75 1 1>}
 ]
}
